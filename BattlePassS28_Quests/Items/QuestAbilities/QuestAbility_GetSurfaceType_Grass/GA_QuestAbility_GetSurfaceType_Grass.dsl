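; Quest gameplay ability that reports time the player spends on grass landscape.
; On activation it creates a paused, looping one-second timer and listens for Event.SurfaceType.Changed.
; A short retriggerable delay debounces a downward line trace, which unpauses the timer only for Grass physical material on Landscape actors or streaming proxies.
; Each tick broadcasts CustomVerb.SurfaceCheck.Grass for the avatar; ability end clears the timer, and debug prints remain.
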(event Ability|EventActivateAbility
  (bind _returnvalue (Utilities|Time|SetTimerbyEvent (delegate Custom|SurfaceCheckTick) 1.0 true))
  (bind _output_get (Variables|Default|SetValidSurfaceTimer _returnvalue))
  (Utilities|Time|PauseTimerbyHandle _output_get)
  ;[0.25 1 0.25] Is Player on Grass?
  (Ability|Async|WaitGameplayEventToActor (Ability|GetAvatarActorfromActorInfo) "(TagName=\"Event.SurfaceType.Changed\")"
    (:then
      _)
    (:EventReceived
      _))
  (Utilities|FlowControl|RetriggerableDelay 0.1)
  (bind _outhit (Collision|LineTraceForObjects (Transformation|GetActorLocation (Ability|GetAvatarActorfromActorInfo)) (+ (Transformation|GetActorLocation (Ability|GetAvatarActorfromActorInfo)) (* (Transformation|GetActorUpVector (Ability|GetAvatarActorfromActorInfo)) -5e+02)) (Utilities|Array|MakeArray "ObjectTypeQuery1") true (Utilities|Array|MakeArray (Ability|GetAvatarActorfromActorInfo)) "None" true "(R=1.000000,G=0.000000,B=0.000000,A=1.000000)" "(R=0.000000,G=1.000000,B=0.000000,A=1.000000)" 1.0))
  (if (and (== (.PhysMat (Collision|BreakHitResult _outhit)) "/Game/FortressPhysicalMaterials/Grass.Grass") (or (== (Utilities|GetClass (.HitActor (Collision|BreakHitResult _outhit))) "/Script/Landscape.Landscape") (== (Utilities|GetClass (.HitActor (Collision|BreakHitResult _outhit))) "/Script/Landscape.LandscapeStreamingProxy")))
    (Utilities|Time|UnpauseTimerbyHandle (Variables|Default|GetValidSurfaceTimer))
    (Development|PrintString "Grass!" false false "(R=0.083333,G=1.000000,B=0.000000,A=1.000000)")
    (else
      (Utilities|Time|PauseTimerbyHandle (Variables|Default|GetValidSurfaceTimer))))
  (Development|PrintString (Utilities|GetDisplayName (.PhysMat (Collision|BreakHitResult _outhit))) false false))

(event Custom|SurfaceCheckTick
  (CustomVerb|BroadcastCustomVerbMessage (CustomVerb|MakeCustomVerbMessage (Ability|GetAvatarActorfromActorInfo) "(TagName=\"CustomVerb.SurfaceCheck.Grass\")"))
  (Development|PrintString "Tick Grass" false false))

(event Ability|EventOnEndAbility
  (params (bWasCancelled :type Boolean))
  (Utilities|Time|ClearandInvalidateTimerbyHandle (Variables|Default|GetValidSurfaceTimer)))
